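# One-shot mono MetaSound for small quadruped companion shake foley, classed under pet creature sounds with companion foley attenuation.
# On play, a shared patch derives a no-repeat count from the wave array, then a random wave is picked from it.
# Chained triggers roll random pitch and decibel offsets within input ranges before the wave player fires, with gain applied by multiply.
# It shares the quadruped footstep concurrency and finishes when the wave ends.
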MetaSoundSource /CosmeticCompanions/Assets/Audio_Shared/Foley/Quadraped/Small/MSS_MImosa_Quadraped_Sm_Shake
inputs [
 "Random Pitch Max (Semitones)" {type=Float}
 "Random Pitch Min (Semitones)" {type=Float}
 "Random Volume Change Max (Decibels)" {type=Float}
 "Random Volume Change Min (Decibels)" {type=Float}
 UE.Source.OnPlay {type=Trigger}
]
interfaces [UE.OutputFormat.Mono:1.0, UE.Source.OneShot:1.0, UE.Source:1.1]
outputs [
 "UE.OutputFormat.Mono.Audio:0" {type=Audio}
 UE.Source.OneShot.OnFinished {type=Trigger}
]
pages [Default {
  input_defaults {
   "Random Pitch Max (Semitones)"=3
   "Random Pitch Min (Semitones)"=-3
   "Random Volume Change Max (Decibels)"=2
   "Random Volume Change Min (Decibels)"=-3
   UE.Source.OnPlay=null
  }
  nodes [
   n1 = VariableAccessor.WaveAsset:Array(Variable=InitVariable.WaveAsset:Array().Variable)
   n2 = /FortMetasound/Patches/MSP_GetNoRepeatNumber(InTrigger=^UE.Source.OnPlay, WaveArray=@n1.Value)
   ArrayRandomGet_0 = Array.Random Get:WaveAsset:Array(No Repeats=@n2.NoRepeatNumber, Next=@n2.OutTrigger, In Array=@n1.Value)
   n3 = UE.RandomFloat(Max=^"Random Pitch Max (Semitones)", Min=^"Random Pitch Min (Semitones)", Next=@ArrayRandomGet_0.On Next)
   n4 = UE.RandomFloat(Max=^"Random Volume Change Max (Decibels)", Min=^"Random Volume Change Min (Decibels)", Next=@n3.On Next)
   DecibelstoLinearGain_0 = UE.Decibels to Linear Gain:Float(Decibels=@n4.Value)
   WavePlayer_0 = UE.Wave Player:Mono(Wave Asset=@ArrayRandomGet_0.Value, Play=@n4.On Next, Pitch Shift=@n3.Value) {also {version [1,1]}}
   Multiply_0 = UE.Multiply:Audio by Float(AdditionalOperands=@DecibelstoLinearGain_0.Linear Gain, PrimaryOperand=@WavePlayer_0.Out Mono)
  ]
  links [{dst="@UE.OutputFormat.Mono.Audio:0", src=Multiply_0.Out}, {dst="@UE.Source.OneShot.OnFinished", src=WavePlayer_0.On Finished}]
 }]
properties {
 assetImportData=&self:AssetImportData
 attenuationSettings=/Game/Sounds/Attenuation/AS_Companion_Foley
 maxDistance=1200
 pitch=1.4
 priority=0
 sampleRate=48000
 soundClassObject=&Fort_Creatures_Pets
 volume=2
 concurrencySet [/CosmeticCompanions/Assets/Audio_Shared/Foley/Quadraped/CONC_Mimosa_Quadraped_Footstep]
}
