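# Sound cue for small quadruped cosmetic companion footsteps, playing hard-surface walk samples as pet creature foley.
# A quality-level node sends its first two slots to a random pick of twelve step waves and its last two to a random pick of the first five.
# A root modulator varies volume and pitch per play, on top of a raised cue pitch.
# Companion foley attenuation and a shared quadruped footstep concurrency set govern playback.
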SoundCue /CosmeticCompanions/Assets/Audio_Shared/Foley/Quadraped/Small/SC_MImosa_Quadraped_Sm_Footsteps
root=Modulator_0
nodes [
 {id=Modulator_0, children [QualityLevel_0], properties {pitchMax=1, pitchMin=0.9, volumeMax=1, volumeMin=0.4}}
 {id=QualityLevel_0, children [Random_0,Random_0,Random_1,Random_1]}
 {
  id=Random_0
  children [
   WavePlayer_12
   WavePlayer_13
   WavePlayer_14
   WavePlayer_15
   WavePlayer_16
   WavePlayer_32
   WavePlayer_33
   WavePlayer_34
   WavePlayer_35
   WavePlayer_36
   WavePlayer_37
   WavePlayer_38
  ]
  properties {weights [1,1,1,1,1,1,1,1,1,1,1,1]}
 }
 {id=Random_1, children [WavePlayer_39, WavePlayer_40, WavePlayer_41, WavePlayer_42, WavePlayer_43], properties {weights [1,1,1,1,1]}}
 {
  id=WavePlayer_12
  properties {soundWaveAssetPtr=/CosmeticCompanions/Assets/Audio_Shared/Foley/Quadraped/Small/Wavs/Step/Mimosa_Quadraped_Sm_Walk_Hard_01}
 }
 {
  id=WavePlayer_13
  properties {soundWaveAssetPtr=/CosmeticCompanions/Assets/Audio_Shared/Foley/Quadraped/Small/Wavs/Step/Mimosa_Quadraped_Sm_Walk_Hard_02}
 }
 {
  id=WavePlayer_14
  properties {soundWaveAssetPtr=/CosmeticCompanions/Assets/Audio_Shared/Foley/Quadraped/Small/Wavs/Step/Mimosa_Quadraped_Sm_Walk_Hard_03}
 }
 {
  id=WavePlayer_15
  properties {soundWaveAssetPtr=/CosmeticCompanions/Assets/Audio_Shared/Foley/Quadraped/Small/Wavs/Step/Mimosa_Quadraped_Sm_Walk_Hard_04}
 }
 {
  id=WavePlayer_16
  properties {soundWaveAssetPtr=/CosmeticCompanions/Assets/Audio_Shared/Foley/Quadraped/Small/Wavs/Step/Mimosa_Quadraped_Sm_Walk_Hard_05}
 }
 {
  id=WavePlayer_32
  properties {soundWaveAssetPtr=/CosmeticCompanions/Assets/Audio_Shared/Foley/Quadraped/Small/Wavs/Step/Mimosa_Quadraped_Sm_Walk_Hard_06}
 }
 {
  id=WavePlayer_33
  properties {soundWaveAssetPtr=/CosmeticCompanions/Assets/Audio_Shared/Foley/Quadraped/Small/Wavs/Step/Mimosa_Quadraped_Sm_Walk_Hard_07}
 }
 {
  id=WavePlayer_34
  properties {soundWaveAssetPtr=/CosmeticCompanions/Assets/Audio_Shared/Foley/Quadraped/Small/Wavs/Step/Mimosa_Quadraped_Sm_Walk_Hard_08}
 }
 {
  id=WavePlayer_35
  properties {soundWaveAssetPtr=/CosmeticCompanions/Assets/Audio_Shared/Foley/Quadraped/Small/Wavs/Step/Mimosa_Quadraped_Sm_Walk_Hard_09}
 }
 {
  id=WavePlayer_36
  properties {soundWaveAssetPtr=/CosmeticCompanions/Assets/Audio_Shared/Foley/Quadraped/Small/Wavs/Step/Mimosa_Quadraped_Sm_Walk_Hard_10}
 }
 {
  id=WavePlayer_37
  properties {soundWaveAssetPtr=/CosmeticCompanions/Assets/Audio_Shared/Foley/Quadraped/Small/Wavs/Step/Mimosa_Quadraped_Sm_Walk_Hard_11}
 }
 {
  id=WavePlayer_38
  properties {soundWaveAssetPtr=/CosmeticCompanions/Assets/Audio_Shared/Foley/Quadraped/Small/Wavs/Step/Mimosa_Quadraped_Sm_Walk_Hard_12}
 }
 {
  id=WavePlayer_39
  properties {soundWaveAssetPtr=&Mimosa_Quadraped_Sm_Walk_Hard_01}
 }
 {
  id=WavePlayer_40
  properties {soundWaveAssetPtr=&Mimosa_Quadraped_Sm_Walk_Hard_02}
 }
 {
  id=WavePlayer_41
  properties {soundWaveAssetPtr=&Mimosa_Quadraped_Sm_Walk_Hard_03}
 }
 {
  id=WavePlayer_42
  properties {soundWaveAssetPtr=&Mimosa_Quadraped_Sm_Walk_Hard_04}
 }
 {
  id=WavePlayer_43
  properties {soundWaveAssetPtr=&Mimosa_Quadraped_Sm_Walk_Hard_05}
 }
]
properties {
 attenuationSettings=/Game/Sounds/Attenuation/AS_Companion_Foley
 maxDistance=1200
 pitchMultiplier=1.5
 priority=0
 soundClassObject=&Fort_Creatures_Pets
 volumeMultiplier=0.8
 concurrencySet [/CosmeticCompanions/Assets/Audio_Shared/Foley/Quadraped/CONC_Mimosa_Quadraped_Footstep]
}
